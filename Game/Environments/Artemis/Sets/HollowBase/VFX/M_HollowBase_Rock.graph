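# Masked, lit rock material for Niagara mesh particles in the HollowBase environment VFX.
# A lava rock base texture is multiplied by particle color or a scaled dynamic parameter, chosen by static switch, then desaturated and tinted by a BaseColor parameter.
# Desaturation defaults to full, so the earlier color input mainly affects brightness.
# The rock normal is angle-corrected blended with a volcano rock normal, intensified through negative flattening.
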Material /Game/Environments/Artemis/Sets/HollowBase/VFX/M_HollowBase_Rock
graph {
 properties {bUsedWithNiagaraMeshParticles=true, blendMode=BLEND_Masked, pixelDepthOffsetMode=PDOM_Legacy, usedShadingModels=Default Lit}
 nodes [
  Constant() {r=0.2}
  Constant() {r=0.7}
  Constant() {r=0.5}
  n1 = TextureSampleParameter2D() {parameterName=Normal, samplerType=SAMPLERTYPE_Normal, texture=/Game/Environments/Artemis/Sets/Terrain/Textures/T_Artemis_Lava_Rock_02_N}
  MaterialFunctionCall(
   BaseNormal=@n1.RGB
   AdditionalNormal=MaterialFunctionCall(
    Normal=TextureSample() {samplerType=SAMPLERTYPE_Normal, texture=/Game/Athena/Environments/Volcano/Textures/Rocks_v2_normal}.RGB
    Flatness=Constant() {r=-1}
   ) {materialFunction=&FlattenNormal}.Result
  ) {materialFunction=&BlendAngleCorrectedNormals}
  n2 = TextureSampleParameter2D() {parameterName=Base, texture=/Game/Environments/Artemis/Sets/Terrain/Textures/T_Artemis_Lava_Rock_02_D}
  n3 = StaticSwitchParameter(
   True=ParticleColor().RGB
   False=Multiply(
    A=ScalarParameter() {defaultValue=1.5, parameterName=BaseColor Multiply}
    B=DynamicParameter() {defaultValue=<rgba 0 1 1 1>, paramNames [Param1,BaseColor,Param3,Param4]}.BaseColor
   )
  ) {parameterName=UseParticleColor?}
  Multiply(
   A=Desaturation(None=Multiply(A=@n2.RGB, B=@n3) {constB=1.5}, Fraction=ScalarParameter() {defaultValue=1, parameterName=Desaturation})
   B=VectorParameter() {parameterName=BaseColor, defaultValue=<rgba 1 1 1 0>}.RGB
  )
 ]
}
